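; Blueprint function library for the CosmicThunder Phase Dash token that reports how many dash charges an Ability System Component has.
; GetPhaseDashCharges reads the CurrentDashCharges attribute from the DashRuntime FortAttributeSet_DashCharges set on the supplied component.
; It returns that value truncated to an integer, or 0 if the component is invalid or the attribute is not found.
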(fn GetPhaseDashCharges
  (params (AbilitySystemComponent :type Ability%20System%20Component%20Object%20Reference))
  (bind (_bsuccessfullyfoundattribute _returnvalue) (Ability|Attribute|GetFloatAttributefromAbilitySystemComponent AbilitySystemComponent "(AttributeName=\"CurrentDashCharges\",Attribute=/Script/DashRuntime.FortAttributeSet_DashCharges:CurrentDashCharges,AttributeOwner=\"/Script/CoreUObject.Class'/Script/DashRuntime.FortAttributeSet_DashCharges'\")"))
  (Utilities|IsValid AbilitySystemComponent
    (:Is%20Valid
      (if _bsuccessfullyfoundattribute
        (return (Math|Float|Truncate _returnvalue))
        (else
          _)))
    (:Is%20Not%20Valid
      _))
  (return 0))
